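; Dual-pistol weapon Blueprint for the Fiend Hunter Dualies that handles muzzle flash effects for both hands and custom fire audio.
; Construction attaches left and right Niagara muzzle components to their weapon meshes' sockets.
; When not scoped, firing activates the chosen hand's effect; non-persistent fire randomizes only the right muzzle's scale, while deactivation stops only the left.
; A fire-sound override spawns attached as a workaround for burst-fire sound design (FORT-96639).
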(event EventPlayWeaponFireCustomAudio
  (params (bPersistentFire :type Boolean) (bSecondaryFire :type Boolean))
  ;[1 1 1 1 32] Fire sound work around to quickly fix problem with this weapon being "burst" but sound design not. FORT-96639
  (Audio|SpawnSoundAttached (Variables|Default|GetFireSoundOverride) (Variables|Transformation|GetRootComponent)))

(fn AddRandomScale
  (params (Offhand :type Boolean))
  (if (<= (Math|Random|RandomFloatinRange 0.0 1) 0.8)
    (if Offhand
      (Transformation|SetRelativeScale3D (Variables|Default|GetLeftMuzzle) (Math|Vector|MakeVector (Math|Random|RandomFloatinRange 0.7 1) (Math|Random|RandomFloatinRange 0.7 1) (Math|Random|RandomFloatinRange 0.7 1)))
      (else
        (Transformation|SetRelativeScale3D (Variables|Default|GetRightMuzzle) (Math|Vector|MakeVector (Math|Random|RandomFloatinRange 0.7 1) (Math|Random|RandomFloatinRange 0.7 1) (Math|Random|RandomFloatinRange 0.7 1)))))
    (elif Offhand
      (Transformation|SetRelativeScale3D (Variables|Default|GetLeftMuzzle) (Math|Vector|MakeVector (Math|Random|RandomFloatinRange 0.5 0.6) (Math|Random|RandomFloatinRange 0.5 0.6) (Math|Random|RandomFloatinRange 0.5 0.6)))
      (else
        (Transformation|SetRelativeScale3D (Variables|Default|GetRightMuzzle) (Math|Vector|MakeVector (Math|Random|RandomFloatinRange 0.5 0.6) (Math|Random|RandomFloatinRange 0.5 0.6) (Math|Random|RandomFloatinRange 0.5 0.6)))))))

(fn DeactivateMuzzleFX
  (Components|Activation|Deactivate (Variables|Default|GetLeftMuzzle)))

(fn MuzzleFlashFX
  (params (Persistent%20Fire :type Boolean) (OffHand :type Boolean))
  (if (not (Weapon|IsUsingScope))
    (if (not Persistent%20Fire)
      (CallFunction|AddRandomScale false))
    (if OffHand
      (Components|Activation|Activate (Variables|Default|GetLeftMuzzle) true)
      (else
        (Components|Activation|Activate (Variables|Default|GetRightMuzzle) true)))
    (Variables|Effects|SetLastPlayFXTime (Utilities|Time|GetGameTimeinSeconds))
    (return)))

(fn ConstructionScript
  (|CallParentFunction)
  (Transformation|AttachComponentToComponent (Variables|Default|GetLeftMuzzle) (CallFunction|GetLeftHandWeaponMesh) (Effects|GetMuzzleSocketName 0) "SnapToTarget" "SnapToTarget" "KeepRelative" true)
  (Transformation|AttachComponentToComponent (Variables|Default|GetRightMuzzle) (Variables|Visual|GetWeaponMesh) (Effects|GetMuzzleSocketName 0) "SnapToTarget" "SnapToTarget" "KeepRelative" true)
  (Niagara|SetNiagaraSystemAsset (Variables|Default|GetLeftMuzzle) (Variables|Effects|GetNiagaraMuzzleSystem) true)
  (Niagara|SetNiagaraSystemAsset (Variables|Default|GetRightMuzzle) (Variables|Effects|GetNiagaraMuzzleSystem) true))
